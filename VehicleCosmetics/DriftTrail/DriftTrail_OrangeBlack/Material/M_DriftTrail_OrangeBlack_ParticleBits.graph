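# Additive, unlit Niagara sprite material for the particle bits of the OrangeBlack drift trail vehicle cosmetic.
# A panning noise-spot texture, tiled by a dynamic parameter, is eroded by subtracting inverted particle alpha, so spots dissolve as alpha drops.
# The result is masked by a soft circle texture, multiplied by particle color and dynamic Brightness, then passed through MF_FX_EmissiveAdjustments.
# The dynamic parameter's color and Desaturate channels are unused.
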Material /VehicleCosmetics/DriftTrail/DriftTrail_OrangeBlack/Material/M_DriftTrail_OrangeBlack_ParticleBits
graph {
 properties {
  bAutomaticallySetUsageInEditor=false
  bUsedWithNiagaraSprites=true
  blendMode=BLEND_Additive
  shadingModel=MSM_Unlit
  usedShadingModels=Unlit
 }
 nodes [
  n1 = DynamicParameter() {paramNames [color,Desaturate,Brightness,Tile]}
  MaterialExpressionVertexColor_0 = ParticleColor()
  n2 = Panner(Coordinate=Multiply(A=@n1.Tile, B=TextureCoordinate() {uTiling=0.2, vTiling=0.2})) {speedX=0.1, speedY=0.1}
  MaterialExpressionConstantClamp_8 = Saturate(None=Subtract(
    A=TextureSample(UVs=@n2) {texture=/DealershipSharedLibrary/FX/Textures/TexturePackage/T_VCVFX_NoiseSpots_RGB}.G
    B=OneMinus(None=@MaterialExpressionVertexColor_0.A)
   ))
  n3 = TextureSample() {texture=/DealershipSharedLibrary/FX/Textures/Gradient_Masks/T_VCVFX_SoftCircle_A}
  MaterialFunctionCall(Emissive=Multiply(A=Multiply(A=@n1.Brightness, B=@MaterialExpressionVertexColor_0.RGB), B=Multiply(A=@MaterialExpressionConstantClamp_8, B=@n3.RGB))) {materialFunction=&MF_FX_EmissiveAdjustments}
 ]
}
