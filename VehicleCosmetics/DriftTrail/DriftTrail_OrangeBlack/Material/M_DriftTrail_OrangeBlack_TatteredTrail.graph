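# Translucent unlit Niagara ribbon material for the tattered trail of the orange-black drift effect.
# Particle color times dynamic Brightness feeds an emissive adjustment function.
# Summed sines and cosines over UV1 and time form a wobble vector, with dynamic R, G, B setting frequency, speed and amplitude, ramped along the ribbon.
# Opacity combines two multiplied panning cloud noises with length and edge fades, then particle alpha, giving the ragged look.
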Material /VehicleCosmetics/DriftTrail/DriftTrail_OrangeBlack/Material/M_DriftTrail_OrangeBlack_TatteredTrail
graph {
 properties {
  bAutomaticallySetUsageInEditor=false
  bUsedWithNiagaraRibbons=true
  blendMode=BLEND_Translucent
  shadingModel=MSM_Unlit
  usedShadingModels=Unlit
 }
 nodes [
  n1 = ParticleColor()
  MaterialFunctionCall(Emissive=Multiply(A=DynamicParameter() {paramNames [Color,Brightness,Color2,BW]}.Brightness, B=@n1.RGB)) {materialFunction=&MF_FX_EmissiveAdjustments}
  n2 = DynamicParameter() {parameterIndex=1, paramNames [R,G,B,A]}
  n3 = ComponentMask(None=Multiply(A=TextureCoordinate() {coordinateIndex=1, uTiling=2}, B=@n2.R)) {r=true}
  n4 = Multiply(A=@n2.G, B=OneMinus(None=Time()))
  n5 = Add(A=@n3, B=@n4)
  n6 = Add(A=Sine(None=@n5), B=Sine(None=@n5) {period=2})
  n7 = Add(A=@n4, B=@n3)
  n8 = MaterialFunctionCall(R=@n6, G=Add(A=Cosine(None=@n7), B=Cosine(None=@n7) {period=2}), B=@n6) {materialFunction=&AppendMany}
  n9 = ComponentMask(None=TextureCoordinate() {uTiling=3}) {r=true}
  Multiply(
   A=Multiply(A=Transform(None=@n8.RGB) {transformSourceType=TRANSFORMSOURCE_World, transformType=TRANSFORM_Local}, B=@n2.B)
   B=Saturate(None=ConstantBiasScale(None=@n9) {bias=-0.4})
  )
  n10 = TextureCoordinate()
  n11 = OneMinus(None=ComponentMask(None=@n10) {r=true})
  n12 = Multiply(A=Constant() {r=3}, B=@n11)
  n13 = ComponentMask(None=@n10) {g=true}
  n14 = Multiply(A=@n13, B=OneMinus(None=@n13))
  n15 = Multiply(A=@n14, B=Constant() {r=5})
  n16 = TextureCoordinate() {coordinateIndex=1, uTiling=0.5, vTiling=0.25}
  n17 = TextureCoordinate() {coordinateIndex=1, vTiling=0.25}
  n18 = Multiply(
   A=TextureSample(UVs=Panner(Coordinate=@n16) {speedX=-1, speedY=0.15}) {texture=/DealershipSharedLibrary/FX/Textures/Noise/T_VCVFX_Clouds_RGB_Soft_03_RGB}.R
   B=TextureSample(UVs=Panner(Coordinate=@n17) {speedX=-1, speedY=0.25}) {texture=/DealershipSharedLibrary/FX/Textures/Noise/T_VCVFX_CloudDistort_A}.G
  )
  n19 = Add(A=Multiply(A=Constant() {r=3}, B=Multiply(A=@n11, B=@n14)), B=ComponentMask(None=Multiply(A=@n15, B=@n18)) {g=true})
  MaterialExpressionConstantClamp_5 = Saturate(None=ConstantBiasScale(None=Multiply(A=@n12, B=@n19)) {bias=-0.25, scale=2})
  Saturate(None=Multiply(A=@n1.A, B=@MaterialExpressionConstantClamp_5))
 ]
}
